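# Translucent, unlit, two-sided particle material for a circular airflow effect on the Cyber Scavenger glider.
# Two noise textures pan in radial coordinates, with dynamic parameters scaling their tiling and speed, and are multiplied together.
# An inverted sphere mask, remapped through a curve atlas row, shapes the noise, which particle alpha then scales.
# Particle color feeds the shared emissive adjustment function.
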Material /BRCosmetics/Gadgets/Assets/VinderTech_GliderChute/Glider_CyberScavenger/FX/M_CyberScavenger_CircleAirFlow
graph {
 properties {
  bUsedWithMeshParticles=true
  bUsedWithNiagaraMeshParticles=true
  blendMode=BLEND_Translucent
  pixelDepthOffsetMode=PDOM_Legacy
  shadingModel=MSM_Unlit
  twoSided=true
  usedShadingModels=Unlit
 }
 nodes [
  n1 = ParticleColor()
  MaterialFunctionCall(Emissive=@n1.RGB) {materialFunction=&MF_FX_EmissiveAdjustments}
  n2 = MaterialFunctionCall() {materialFunction=&VectorToRadialValue}
  n3 = DynamicParameter() {paramNames [UV_tile,SpeedMult,Param3,Param4]}
  n4 = Multiply(
   A=ComponentMask(None=VectorParameter() {group=UV_tile, parameterName=UV_tile1, defaultValue=<rgba 1 0.3 0 1>}.RGB) {g=true, r=true}
   B=@n3.UV_tile
  )
  n5 = ComponentMask(None=VectorParameter() {group=UV_speed, parameterName=Speed_UV1, defaultValue=<rgba 0.3 1 0 1>}.RGB) {g=true, r=true}
  n6 = TextureSampleParameter2D(UVs=Panner(Coordinate=Multiply(A=@n2.Radial Coordinates, B=@n4), Speed=Multiply(A=@n5, B=@n3.SpeedMult)) {speedY=1}) {parameterName=Texture1, texture=/Game/Weapons/WeaponSkins/Textures/T_Swirly_Noise_D1}
  n7 = Multiply(
   A=@n3.UV_tile
   B=ComponentMask(None=VectorParameter() {group=UV_tile, parameterName=UV_tile2, sortPriority=1, defaultValue=<rgba 2 0.3 0 1>}.RGB) {g=true, r=true}
  )
  n8 = ComponentMask(None=VectorParameter() {group=UV_speed, parameterName=Speed_UV2, sortPriority=1, defaultValue=<rgba 0.6 0.5 0 1>}.RGB) {g=true, r=true}
  n9 = TextureSampleParameter2D(UVs=Panner(Coordinate=Multiply(A=@n2.Radial Coordinates, B=@n7), Speed=Multiply(A=@n3.SpeedMult, B=@n8)) {speedY=1}) {parameterName=Texture2, samplerType=SAMPLERTYPE_LinearColor, texture=/Game/Global/Textures/Default/Noise/T_Noise01_Inverse}
  n10 = TextureCoordinate()
  n11 = Constant2Vector() {g=0.5, r=0.5}
  n12 = Constant() {r=0.5}
  n13 = Constant()
  n14 = CurveAtlasRowParameter(CurveTime=OneMinus(None=SphereMask(A=@n10, B=@n11, Radius=@n12, Hardness=@n13))) {
   atlas=&AthenaCurveAtlas01
   curve=/Game/Gadgets/Assets/VinderTech_GliderChute/Glider_CyberScavenger/FX/Curve_CyberScavenger_Glider
   defaultValue=89
   parameterName=Sphere_Mask_Ajustment
  }
  Multiply(A=@n1.A, B=Multiply(A=Multiply(A=@n6.R, B=@n9.R), B=@n14.G))
 ]
}
