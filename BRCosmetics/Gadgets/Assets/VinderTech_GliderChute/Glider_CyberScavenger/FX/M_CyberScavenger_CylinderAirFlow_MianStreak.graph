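# Translucent, unlit, two-sided particle material for the main airflow streak effect on the CyberScavenger glider.
# Particle color drives emissive through MF_FX_EmissiveAdjustments.
# Opacity combines particle alpha, a curve-atlas fade along inverted V, and a mirrored-U falloff that peaks at the center, sharpened with Power nodes.
# The default shading path hides the streak when camera Z exceeds particle Z plus an offset; mobile skips this check.
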Material /BRCosmetics/Gadgets/Assets/VinderTech_GliderChute/Glider_CyberScavenger/FX/M_CyberScavenger_CylinderAirFlow_MianStreak
graph {
 properties {
  bUsedWithParticleSprites=true
  blendMode=BLEND_Translucent
  pixelDepthOffsetMode=PDOM_Legacy
  shadingModel=MSM_Unlit
  twoSided=true
  usedShadingModels=Unlit
 }
 nodes [
  n1 = ParticleColor()
  MaterialFunctionCall(Emissive=@n1.RGB) {materialFunction=&MF_FX_EmissiveAdjustments}
  n2 = TextureCoordinate()
  n3 = CurveAtlasRowParameter(CurveTime=OneMinus(None=ComponentMask(None=@n2) {g=true})) {
   atlas=&AthenaCurveAtlas01
   curve=/Game/Gadgets/Assets/VinderTech_GliderChute/Glider_CyberScavenger/FX/Curve_CyberScavenger_Glider
   defaultValue=89
   parameterName=Streak_Fade
  }
  n4 = ComponentMask(None=@n2) {r=true}
  n5 = Constant() {r=0.5}
  n6 = LinearInterpolate(Alpha=@n4) {constA=1, constB=0}
  n7 = Power(Base=Multiply(A=If(A=@n4, B=@n5, "A > B"=@n6, "A == B"=@n4, "A < B"=@n4)) {constB=2})
  n8 = Multiply(A=@n1.A, B=Power(Base=Multiply(A=@n3.B, B=@n7)))
  n9 = ComponentMask(None=CameraPositionWS()) {b=true}
  n10 = Add(A=ComponentMask(None=ParticlePositionWS()) {b=true}) {constB=300}
  n11 = Constant()
  n12 = Constant() {r=1}
  ShadingPathSwitch(Default=Multiply(A=@n8, B=If(A=@n9, B=@n10, "A > B"=@n11, "A == B"=@n12, "A < B"=@n12)), Mobile=@n8) {desc=Migrated from FeatureLevelSwitch}
 ]
}
