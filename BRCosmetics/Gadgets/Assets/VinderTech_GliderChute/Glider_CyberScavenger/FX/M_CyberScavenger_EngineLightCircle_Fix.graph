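# Translucent, unlit, two-sided mesh-particle material for the circular engine light effect on the CyberScavenger glider.
# The red channel of the circle texture masks the glow, with a Panner parameter blending between static and scrolling UVs.
# A repeating clamped ramp, scrolled by time and the CLightSpeed dynamic parameter, creates moving light bands tinted by particle color.
# These add onto a constant red base and feed the shared emissive adjustments function.
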Material /BRCosmetics/Gadgets/Assets/VinderTech_GliderChute/Glider_CyberScavenger/FX/M_CyberScavenger_EngineLightCircle_Fix
graph {
 properties {
  bUsedWithMeshParticles=true
  blendMode=BLEND_Translucent
  pixelDepthOffsetMode=PDOM_Legacy
  shadingModel=MSM_Unlit
  twoSided=true
  usedShadingModels=Unlit
 }
 nodes [
  Constant() {r=1}
  n1 = ParticleColor()
  n2 = TextureCoordinate()
  n3 = TextureSample!(
   UVs=LinearInterpolate(A=@n2, B=Panner(Coordinate=@n2) {speedX=-0.2}, Alpha=ScalarParameter() {parameterName=Panner})
   Tex
   Level
   Apply View MipBias
  ) {
   mipValueMode=TMVM_MipLevel
   texture=&/T_CyberScavenger_CircleEngine_light
  }
  n4 = Multiply(A=TextureCoordinate()) {constB=2}
  n5 = Multiply(A=Multiply(A=Time()) {constB=-1}, B=DynamicParameter() {paramNames [CLightSpeed,Param2,Param3,Param4]}.CLightSpeed) {constB=-1}
  n6 = Frac(None=ComponentMask(None=Add(A=@n4, B=@n5)) {r=true})
  n7 = Multiply(A=@n3.R, B=Clamp(None=LinearInterpolate(Alpha=@n6) {constA=-1})) {constA=1}
  n8 = Constant3Vector() {constant=<rgba 1 0 0 0>}
  MaterialFunctionCall(Emissive=Add(A=Multiply(A=@n1.RGB, B=@n7), B=Multiply(A=@n3.R, B=@n8)) {constB=0}) {materialFunction=&MF_FX_EmissiveAdjustments}
 ]
}
