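# Translucent, unlit, two-sided mesh-particle material for the glowing engine light tail of the CyberScavenger glider.
# The texture's inverted blue channel drives a BumpOffset parallax for the green-channel mask.
# Blue also offsets a time-scrolled pulse, with speed from Dynamic Parameter BrakeLightSpeed, tinted by particle color over a constant red base.
# A blue-channel power darkens depth and Fresnel fades edges before brightness scaling feeds MF_FX_EmissiveAdjustments.
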Material /BRCosmetics/Gadgets/Assets/VinderTech_GliderChute/Glider_CyberScavenger/FX/M_CyberScavenger_EngineLightTail
graph {
 properties {
  bUsedWithMeshParticles=true
  blendMode=BLEND_Translucent
  pixelDepthOffsetMode=PDOM_Legacy
  shadingModel=MSM_Unlit
  twoSided=true
  usedShadingModels=Unlit
 }
 nodes [
  n1 = ParticleColor()
  n2 = TextureSample!(UVs, Tex, Level, Apply View MipBias) {
   mipValueMode=TMVM_MipLevel
   samplerSource=SSM_Clamp_WorldGroupSettings
   texture=&/T_CyberScavenger_CircleEngine_light
  }
  n3 = OneMinus(None=@n2.B)
  n4 = TextureSample(UVs=BumpOffset(
    Coordinate=TextureCoordinate()
    Height=ScalarParameter() {defaultValue=0.66, parameterName=Height}
    HeightRatioInput=ConstantBiasScale(None=@n3) {bias=-0.5, scale=1}
   )) {
   mipValueMode=TMVM_MipLevel
   samplerSource=SSM_Clamp_WorldGroupSettings
   texture=&T_CyberScavenger_CircleEngine_light
  }
  n5 = Multiply(A=@n4.G, B=@n2.B)
  n6 = Time()
  n7 = DynamicParameter() {paramNames [BrakeLightSpeed, Param2, Param3, Param4]}
  n8 = Frac(None=Add(A=@n2.B, B=Multiply(A=@n6, B=@n7.BrakeLightSpeed)))
  n9 = Multiply(A=@n5, B=Clamp(None=LinearInterpolate(Alpha=@n8) {constA=-1.5, constB=1.5})) {constA=1}
  n10 = Constant3Vector() {constant=<rgba 1 0 0 0>}
  n11 = Multiply(
   A=Add(A=Multiply(A=@n1.RGB, B=@n9), B=Multiply(A=@n5, B=@n10)) {constB=0}
   B=Power(Base=@n2.B, Exp=ScalarParameter() {defaultValue=2, parameterName=DarkenDepthEffect})
  )
  n12 = Fresnel(ExponentIn=ScalarParameter() {defaultValue=0.25, parameterName=LightEdgeEX}) {baseReflectFraction=0}
  MaterialFunctionCall(Emissive=Multiply(A=ScalarParameter() {defaultValue=5, parameterName=EmissiveBrightness}, B=LinearInterpolate(A=@n11, Alpha=@n12) {constB=0})) {materialFunction=&MF_FX_EmissiveAdjustments}
 ]
}
